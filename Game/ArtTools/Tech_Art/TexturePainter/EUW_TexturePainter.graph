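# Editor utility widget that paints brush strokes onto a texture and applies it to a skeletal mesh.
# Load copies the source image into a render target; Tick stamps brush alphas at the mouse, once per click or repeatedly in airbrush mode.
# Manual or on-release auto-save exports a temporary PNG and reimports it over the source via Interchange.
# Apply exits paint mode and assigns a dynamic material instance using the source texture.
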EditorUtilityWidgetBlueprint /Game/ArtTools/Tech_Art/TexturePainter/EUW_TexturePainter
graphs [
 EventGraph {nodes [
   "On Clicked (LoadSource)"!() -> (OutputDelegate, then, widget:User Widget Object Reference, event:Pointer Event Structure) {is=title K2Node_ComponentBoundEvent}
   > Set Visibility(Visibility=Visible, self=Get BrushControls() -> (BrushControls:Details View Object Reference))
   > Set IsTexturePaintMode?(IsTexturePaintMode?:bool=true) -> (Output_Get:bool)
   > n1 = Sequence()
   > UpdateRenderTarget!(execute, self, RenderTexture:Texture Object Reference=Get SourceImage() -> (SourceImage:Texture Object Reference)) -> (then)
   SetBrush!(
    execute=@n1.then_1
    self=Get Workimage() -> (Workimage:Image Object Reference)
    InBrush_TintColor:Slate Color Structure=(SpecifiedColor=(R=1,G=1,B=1,A=1),ColorUseRule=UseColor_Specified)
    InBrush_DrawAs:ESlateBrushDrawType Enum=Image
    InBrush_Tiling:ESlateBrushTileType Enum=NoTile
    InBrush_Mirroring:ESlateBrushMirrorType Enum=NoMirror
    InBrush_ImageSize:Vector 2D Structure=(X=32,Y=32)
    InBrush_Margin:Margin Structure=(Left=0,Top=0,Right=0,Bottom=0)
    InBrush_ResourceObject:Object Reference=Get RenderTarget() -> (RenderTarget:Canvas Render Target 2D Object Reference)
    InBrush_OutlineSettings:Slate Brush Outline Settings Structure=(CornerRadii=(X=0,Y=0,Z=0,W=0),Color=(SpecifiedColor=(R=0,G=0,B=0,A=0),ColorUseRule=UseColor_Specified),Width=0,RoundingType=HalfHeightRadius,bUseBrushTransparency=False)
   ) -> (then)
   "On Clicked (ApplyToSkelMesh)"!() -> (OutputDelegate, then, widget:User Widget Object Reference, event:Pointer Event Structure) {is=title K2Node_ComponentBoundEvent}
   > Set Visibility(Visibility=Hidden, self=Get BrushControls())
   > Set IsTexturePaintMode?(IsTexturePaintMode?=false)
   > n2 = CreateDynamicMaterialInstance!(
    execute
    self:Primitive Component Object Reference=Get SkeletalMeshComponent!(self:Skeletal Mesh Actor Object Reference=Get SkelMeshTarget() -> (SkelMeshTarget:Skeletal Mesh Actor Object Reference)) -> (SkeletalMeshComponent:Skeletal Mesh Component Object Reference)
    ElementIndex:int=0
    SourceMaterial:Material Interface Object Reference=&/M_TexturePaintMaterial_MeshPainter
    OptionalName:Name=None
   ) -> (then, ReturnValue:Material Instance Dynamic Object Reference) {is=title K2Node_CallFunction}
   > SetTextureParameterValue(self=@n2, ParameterName=Texture, Value=Get SourceImage())
   > SetBrush(
    self=Get Workimage()
    InBrush_TintColor=(SpecifiedColor=(R=1,G=1,B=1,A=1),ColorUseRule=UseColor_Specified)
    InBrush_DrawAs=Image
    InBrush_Tiling=NoTile
    InBrush_Mirroring=NoMirror
    InBrush_ImageSize=(X=32,Y=32)
    InBrush_Margin=(Left=0,Top=0,Right=0,Bottom=0)
    InBrush_ResourceObject=Get SourceImage()
    InBrush_OutlineSettings=(CornerRadii=(X=0,Y=0,Z=0,W=0),Color=(SpecifiedColor=(R=0,G=0,B=0,A=0),ColorUseRule=UseColor_Specified),Width=0,RoundingType=HalfHeightRadius,bUseBrushTransparency=False)
   )
   "On Clicked (ManualSave)"!() -> (OutputDelegate, then, widget:User Widget Object Reference, event:Pointer Event Structure) {is=title K2Node_ComponentBoundEvent}
   > ReimportToTexture()
   n3 = ReimportToTexture() {is=title K2Node_CustomEvent}
   n4 = Append(A=Get!() -> (SaveDirectory_Path:String), B=/TempTransfere.png)
   n5 = Reimport!() -> (OutputDelegate, then, bSuccess:bool=~) {is=title K2Node_CustomEvent}
   ExportToDisk!(
    execute=@n3.then
    Texture:Texture Object Reference=Get RenderTarget()
    Filename:String=@n4
    Options_Format:EDesiredImageFormat Enum=PNG
    Options_OnComplete:Delegate=@n5
    Options_CompressionQuality:int=0
    Options_bOverwriteFile:bool=True
    Options_bAsync:bool=True
   ) -> (then)
   Event Construct()
   > SetObject(self=[Get Controls() -> (Controls:Details View Object Reference), Get BrushControls()], NewObject=Self-Reference())
   n6 = Event Tick()
   > n7 = GetMousePositionOverImage!(execute, self, Geometry:"Geometry Structure (by ref)"=@n6.MyGeometry) -> (then, ReturnValue:Vector 2D Structure)
   > n8 = Branch(Condition=Get MouseIsDown?() -> (MouseIsDown?:bool))
   > n9 = Branch(Condition=Get IsAirbrushMode?() -> (IsAirbrushMode?:bool))
   > n10 = Delay!(execute, WorldContextObject:Object Reference, Duration=Get BrushPaintFrequency() -> (BrushPaintFrequency:float)) -> (then)
   n11 = BeginDrawCanvasToRenderTarget!(
    execute=[@n10.then, @n13.Completed]
    WorldContextObject:Object Reference
    TextureRenderTarget:Texture Render Target 2D Object Reference=Get RenderTarget()
   ) -> (then, Canvas:Canvas Object Reference, Size:Vector 2D Structure, Context:Draw To Render Target Context Structure)
   n12 = Get BrushRadius() -> (BrushRadius:double)
   > Draw Texture!(
    execute
    self:Canvas Object Reference=@n11
    RenderTexture:Texture Object Reference=Get BrushAlpha() -> (BrushAlpha:Texture Object Reference)
    ScreenPosition:Vector 2D Structure=vector2d - vector2d(
     A:Vector 2D Structure=vector2d * vector2d(A:Vector 2D Structure=@n11.Size, B:Vector 2D Structure=@n7) -> (ReturnValue:Vector 2D Structure) {is=title K2Node_PromotableOperator}
     B:Vector 2D Structure=vector2d / vector2d(A=@n12, B=2) -> (ReturnValue:Vector 2D Structure) {is=title K2Node_PromotableOperator}
    ) -> (ReturnValue:Vector 2D Structure) {is=title K2Node_PromotableOperator}
    ScreenSize:Vector 2D Structure=vector2d * vector2d(A=Get Brush Shape() -> (Brush Shape:Vector 2D Structure), B:double=@n12)
    CoordinatePosition:Vector 2D Structure=(X=0,Y=0)
    CoordinateSize:Vector 2D Structure=(X=1,Y=1)
    RenderColor:Linear Color Structure=Get BrushColor() -> (BrushColor:Linear Color Structure)
    BlendMode:EBlendMode Enum=Get Blend Mode() -> (Blend Mode:EBlendMode Enum=BLEND_Opaque)
    Rotation:float=RandomFloatInRange(Min=0, Max=Get MaxRandomRotation() -> (MaxRandomRotation:double))
    PivotPoint:Vector 2D Structure=(X=0.5,Y=0.5)
   ) -> (then)
   > EndDrawCanvasToRenderTarget!(execute, WorldContextObject:Object Reference, Context:"Draw To Render Target Context Structure (by ref)"=@n11.Context) -> (then)
   > Set TriggerUpdateToStaticTexture?(TriggerUpdateToStaticTexture?:bool=true) -> (Output_Get:bool)
   n13 = Do Once(execute=@n9.else, Reset=@n14.then_0)
   n14 = Sequence(execute=@n8.else)
   Branch(
    execute=@n14.then_1
    Condition=AND Boolean!(
     A=Get TriggerUpdateToStaticTexture?() -> (TriggerUpdateToStaticTexture?:bool)
     B=Get IsTexturePaintMode?() -> (IsTexturePaintMode?:bool)
     C:bool=Get AutoSaveTexture?() -> (AutoSaveTexture?:bool)
    ) -> (ReturnValue)
   )
   > Set TriggerUpdateToStaticTexture?(TriggerUpdateToStaticTexture?=false)
   > ReimportToTexture() {is=title K2Node_CallFunction}
   PrintString(execute=@n18.else, InString=Export of OutputTarget Failed)
   n15 = CreateSourceData!(execute=@n17.then, InFileName:String=@n4) -> (then, ReturnValue:Interchange Source Data Object Reference)
   > n16 = GetInterchangeManagerScripted!(execute) -> (then, ReturnValue:Interchange Manager Object Reference)
   > ImportAsset!(
    execute
    self:Interchange Manager Object Reference=@n16
    ContentPath:String
    SourceData:Interchange Source Data Object Reference=@n15
    ImportAssetParameters:"Import Asset Parameters Structure (by ref)"=Make Import Asset Parameters!(
     ReimportAsset:Object Reference=Get SourceImage()
     ReimportSourceIndex:int=-1
     bIsAutomated:bool=False
     bFollowRedirectors:bool=False
     OverridePipelines:Array of Soft Object Path Structures
     ImportLevel:Level Object Reference
     DestinationName:String
     bReplaceExisting:bool=True
     bForceShowDialog:bool=False
    ) -> (ImportAssetParameters:Import Asset Parameters Structure) {is=title K2Node_MakeStruct}
   ) -> (then, OutImportedObjects:Array of Object References, ReturnValue:bool)
   Event PreConstruct()
   Get RenderTarget()
   n17 = DelayUntilNextTick!(execute=@n18.then, WorldContextObject:Object Reference) -> (then)
   n18 = Branch(execute=@n5.then, Condition=@n5.bSuccess)
  ]}
 GetMousePositionOverImage {nodes [
   n1 = GetMousePositionOverImage!() -> (then, Geometry:"Geometry Structure (by ref)") {is=title K2Node_FunctionEntry}
   > n2 = GetMousePositionOnPlatform!(execute) -> (then, ReturnValue:Vector 2D Structure)
   n3 = GetLocalSize!(Geometry:"Geometry Structure (by ref)"=GetPaintSpaceGeometry!(self:Widget Object Reference=Get Workimage()) -> (ReturnValue:"Geometry Structure (by ref)")) -> (ReturnValue:Vector 2D Structure)
   n4 = vector2d + vector2d(
    A:Vector 2D Structure=vector2d / vector2d(A:Vector 2D Structure=@n2, B:float=GetViewportScale!(WorldContextObject:Object Reference) -> (ReturnValue:float))
    B:Vector 2D Structure=ScreenToLocal!(
     WorldContextObject:Object Reference
     Geometry:"Geometry Structure (by ref)"=@n1
     ScreenPosition:Vector 2D Structure
     bIncludeWindowPosition:bool=false
    ) -> (LocalCoordinate:Vector 2D Structure)
   ) -> (ReturnValue:Vector 2D Structure) {is=title K2Node_PromotableOperator}
   > Return Node!(
    execute
    ReturnValue:Vector 2D Structure=ClampAxes2D!(A:Vector 2D Structure=vector2d / vector2d(A=@n4, B:Vector 2D Structure=@n3), MinAxisVal:double=0, MaxAxisVal:double=1) -> (ReturnValue:Vector 2D Structure)
   )
  ]}
 OnMouseButtonDown {nodes [
   OnMouseButtonDown!() -> (then, MyGeometry:Geometry Structure, MouseEvent:"Pointer Event Structure (by ref)") {is=title K2Node_FunctionEntry}
   > Set MouseIsDown?(MouseIsDown?:bool=true) -> (Output_Get:bool)
   > Return Node(ReturnValue:Event Reply Structure=Unhandled())
  ]}
 OnMouseButtonUp {nodes [
   OnMouseButtonUp!() -> (then, MyGeometry:Geometry Structure, MouseEvent:"Pointer Event Structure (by ref)") {is=title K2Node_FunctionEntry}
   > Set MouseIsDown?(MouseIsDown?=false)
   > Return Node(ReturnValue=Unhandled())
  ]}
 UpdateRenderTarget {nodes [
   n1 = UpdateRenderTarget!() -> (then, RenderTexture:Texture Object Reference) {is=title K2Node_FunctionEntry}
   > n2 = BeginDrawCanvasToRenderTarget(TextureRenderTarget=Get RenderTarget())
   > Draw Texture(
    self=@n2
    RenderTexture=@n1
    ScreenSize=@n2.Size
    CoordinateSize=(X=1,Y=1)
    RenderColor=(R=1,G=1,B=1,A=1)
    BlendMode=BLEND_Opaque
    Rotation=0
    PivotPoint=(X=0.5,Y=0.5)
   )
   > EndDrawCanvasToRenderTarget(Context=@n2.Context)
  ]}
]
variables [
 AutoSaveTexture? {type=boolean}
 Blend Mode {type=EBlendMode, default=BLEND_AlphaComposite}
 Brush Shape {type=Vector2D, default=<1 1>}
 BrushAlpha {type=Texture, default=/Game/ArtTools/Tech_Art/TexturePainter/Brushes/T_New_Soft_Flair_Alpha}
 BrushColor {type=LinearColor, default=<rgba 1 1 1 1>}
 BrushPaintFrequency {type=number, default=0.01}
 BrushRadius {type=number, default=30}
 IsAirbrushMode? {type=boolean}
 IsTexturePaintMode? {type=boolean}
 MaxRandomRotation {type=number}
 MouseIsDown? {type=boolean}
 RenderTarget {type=CanvasRenderTarget2D, default=&/CRT_TexturePaintTarget}
 SaveDirectory {type=DirectoryPath, default {path=C:/BIOMEMASKINGTEMPFILETRANSFEREFOLDER}}
 SkelMeshTarget {type=SkeletalMeshActor, default=null}
 SourceImage {type=Texture, default=/Game/Athena/UI/InGame/Sidekick/Reactions/100}
 StaticTextureSource {type=Texture, default=&/T_TexturePainter_Static}
 TriggerUpdateToStaticTexture? {type=boolean}
]
widgets {
 info {rootWidgetClass=&Overlay, widgetCount=10}
 widgets [
  Overlay_38 {widgetClassPath=&Overlay, widget=Overlay_38}
  HorizontalBox_0 {
   widgetClassPath=&HorizontalBox
   parent=Overlay_38
   widget=HorizontalBox_0
   slot {class=OverlaySlot, properties {horizontalAlignment=HAlign_Fill, verticalAlignment=VAlign_Fill}}
  }
  Workimage {
   widgetClassPath=&Image
   bIsVariable=true
   parent=HorizontalBox_0
   widget=Workimage
   properties {brush {
     bIsDynamicallyLoaded=false
     drawAs=Image
     imageType=NoImage
     mirroring=NoMirror
     resourceName=None
     resourceObject=/Game/Global/Textures/Default/Masks_and_Gradients/16x
     tiling=NoTile
     imageSize=<64 64>
     margin=<ltrb 0 0 0 0>
     outlineSettings {
      bUseBrushTransparency=false
      roundingType=HalfHeightRadius
      width=0
      color {colorUseRule=UseColor_Specified, specifiedColor=<rgba 0 0 0 0>}
      cornerRadii=<0 0 0 0>
     }
     tintColor {colorUseRule=UseColor_Specified, specifiedColor=<rgba 1 1 1 1>}
     uVRegion {bIsValid=false, max=<0 0>, min=<0 0>}
    }}
   slot {class=HorizontalBoxSlot, properties {size {sizeRule=Fill, value=1}}}
  }
  VerticalBox_0 {
   widgetClassPath=&VerticalBox
   parent=HorizontalBox_0
   widget=VerticalBox_0
   slot {class=HorizontalBoxSlot, properties {horizontalAlignment=HAlign_Right}}
  }
  HorizontalBox_1 {
   widgetClassPath=&HorizontalBox
   parent=VerticalBox_0
   widget=HorizontalBox_1
   slot {class=VerticalBoxSlot, properties {horizontalAlignment=HAlign_Right}}
  }
  LoadSource {
   widgetClassPath=&TA_Generic01_Button
   parent=HorizontalBox_1
   widget=LoadSource
   properties {
    label=Load Source Image
    iconBrush {
     bIsDynamicallyLoaded=false
     drawAs=Image
     imageType=NoImage
     mirroring=NoMirror
     resourceName=None
     resourceObject=null
     tiling=NoTile
     imageSize=<2199023779840 0>
     margin=<ltrb 0 0 0 0>
     outlineSettings {
      bUseBrushTransparency=false
      roundingType=HalfHeightRadius
      width=0
      color {colorUseRule=UseColor_Specified, specifiedColor=<rgba 0 0 0 0>}
      cornerRadii=<0 0 0 1>
     }
     tintColor {colorUseRule=UseColor_Specified, specifiedColor=<rgba 1 1 1 1>}
     uVRegion {bIsValid=false, max=<0 0>, min=<0 0>}
    }
   }
   slot {class=HorizontalBoxSlot}
  }
  ManualSave {
   widgetClassPath=&TA_Generic01_Button
   parent=HorizontalBox_1
   widget=ManualSave
   properties {
    label=ManualUpdateToSource
    iconBrush {
     bIsDynamicallyLoaded=false
     drawAs=Image
     imageType=NoImage
     mirroring=NoMirror
     resourceName=None
     resourceObject=null
     tiling=NoTile
     imageSize=<2199023779840 0>
     margin=<ltrb 0 0 0 0>
     outlineSettings {
      bUseBrushTransparency=false
      roundingType=HalfHeightRadius
      width=0
      color {colorUseRule=UseColor_Specified, specifiedColor=<rgba 0 0 0 0>}
      cornerRadii=<0 0 0 1>
     }
     tintColor {colorUseRule=UseColor_Specified, specifiedColor=<rgba 1 1 1 1>}
     uVRegion {bIsValid=false, max=<0 0>, min=<0 0>}
    }
   }
   slot {class=HorizontalBoxSlot}
  }
  ApplyToSkelMesh {
   widgetClassPath=&TA_Generic01_Button
   parent=VerticalBox_0
   widget=ApplyToSkelMesh
   properties {
    label=ApplyToMesh
    iconBrush {
     bIsDynamicallyLoaded=false
     drawAs=Image
     imageType=NoImage
     mirroring=NoMirror
     resourceName=None
     resourceObject=null
     tiling=NoTile
     imageSize=<2199023779840 0>
     margin=<ltrb 0 0 0 0>
     outlineSettings {
      bUseBrushTransparency=false
      roundingType=HalfHeightRadius
      width=0
      color {colorUseRule=UseColor_Specified, specifiedColor=<rgba 0 0 0 0>}
      cornerRadii=<0 0 0 1>
     }
     tintColor {colorUseRule=UseColor_Specified, specifiedColor=<rgba 1 1 1 1>}
     uVRegion {bIsValid=false, max=<0 0>, min=<0 0>}
    }
   }
   slot {class=VerticalBoxSlot}
  }
  Controls {
   widgetClassPath=&DetailsView
   parent=VerticalBox_0
   widget=Controls
   properties {categoriesToShow [NoCategory]}
   slot {class=VerticalBoxSlot, properties {horizontalAlignment=HAlign_Right}}
  }
  BrushControls {
   widgetClassPath=&DetailsView
   parent=VerticalBox_0
   widget=BrushControls
   properties {visibility=Hidden, categoriesToShow [Brush]}
   slot {class=VerticalBoxSlot, properties {horizontalAlignment=HAlign_Right}}
  }
 ]
}
