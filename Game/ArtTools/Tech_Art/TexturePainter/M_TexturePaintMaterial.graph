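# Unlit material in the tech-art TexturePainter tool that displays a paint texture directly on meshes.
# A texture parameter named Texture samples CRT_TexturePaintTarget by default as linear color and feeds its RGB into emissive through Make Material Attributes, so scene lighting does not affect it.
# Usage flags cover skeletal meshes, morph targets, mesh deformers and Nanite.
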Material /Game/ArtTools/Tech_Art/TexturePainter/M_TexturePaintMaterial
graph {
 properties {
  bUseMaterialAttributes=true
  bUsedWithMeshDeformer=true
  bUsedWithMorphTargets=true
  bUsedWithNanite=true
  bUsedWithSkeletalMesh=true
  pixelDepthOffsetMode=PDOM_Legacy
  refractionMethod=RM_IndexOfRefraction
  shadingModel=MSM_Unlit
  usedShadingModels=Unlit
 }
 nodes [
  n1 = TextureSampleParameter2D() {parameterName=Texture, samplerType=SAMPLERTYPE_LinearColor, texture=&/CRT_TexturePaintTarget}
  MakeMaterialAttributes!(
   BaseColor
   Metallic
   Specular
   Roughness
   Anisotropy
   EmissiveColor=@n1.RGB
   Opacity
   OpacityMask
   Normal
   Tangent
   WorldPositionOffset
   SubsurfaceColor
   ClearCoat
   ClearCoatRoughness
   AmbientOcclusion
   Refraction
   CustomizedUVs_0
   CustomizedUVs_1
   CustomizedUVs_2
   CustomizedUVs_3
   CustomizedUVs_4
   CustomizedUVs_5
   CustomizedUVs_6
   CustomizedUVs_7
   PixelDepthOffset
   ShadingModel
   Displacement
  )
 ]
}
